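# Conversation database for the Brainiac NPC, a semi-conscious zombie that greets the player when talked to.
# The entry point leads to one Speech node whose default greeting can be replaced by contextual lines.
# Controller requirements select these lines for a first-ever conversation or for specific character cosmetics, with a curve-driven priority on the first-meeting line.
# The speech then links out to the shared NPC library Quests and BasicServices conversations by entry tag.
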ConversationDatabase /NPCLibrary/NPCs/Brainiac/Brainiac/NPCLibrary_Conversation_Brainiac
entry_points [{tag=Conversation.Entry.Brainiac, destinations [#1]}]
nodes [
 ConversationEntryPointNode_7 {
  id=#1
  class=ConversationEntryPointNode
  outputs [#2]
  properties {entryTag {tagName=Conversation.Entry.Brainiac}}
 }
 Speech_C_4 {
  id=#2
  class=Speech_C
  outputs [#3, #4]
  properties {
   nodeName=""
   generalConfig {
    defaultMessage="Hello, Looper! Good to see--BGUUUURHGHHG."
    contextualMessages [
     {
      message=I'm conscious again... But for how long?
      contextRequirements [{requirement {class=FortControllerRequirement_IsFirstEverConversationWithNPC}}]
      priority {curve {rowName=Tandem.Speech.Priority.FirstEverInteraction}}
     }
     {
      message="My zombie mind could return at any ti---GUUUUHGR!"
      contextRequirements [{requirement {
         class=FortControllerRequirement_HasCID
         properties {allowedCIds [/BRCosmetics/Athena/Items/Cosmetics/Characters/CID_A_373_Athena_Commando_M_OriginPrisoner]}
        }}]
      priority {value=0, curve {curveTable=null, rowName=None}}
     }
     {
      message=Trust me. You don't want a semi-conscious zombie on the battlefield.
      contextRequirements [{requirement {
         class=FortControllerRequirement_HasCID
         properties {allowedCIds [/BRCosmetics/Athena/Items/Cosmetics/Characters/CID_A_374_Athena_Commando_F_Binary]}
        }}]
      priority {value=0, curve {curveTable=null, rowName=None}}
     }
     {
      message=GuuuurghhHg...
      contextRequirements [{requirement {
         class=FortControllerRequirement_HasCID
         properties {allowedCIds [/BRCosmetics/Athena/Items/Cosmetics/Characters/CID_A_370_Athena_Commando_M_OrderGuard]}
        }}]
      priority {value=0, curve {curveTable=null, rowName=None}}
     }
    ]
   }
  }
 }
 ConversationLinkNode_0 {id=#4, class=ConversationLinkNode, properties {remoteEntryTag {tagName=Conversation.Entry.BasicServices}}}
 ConversationLinkNode_1 {id=#3, class=ConversationLinkNode, properties {remoteEntryTag {tagName=Conversation.Entry.NPCLibrary.Quests}}}
]
